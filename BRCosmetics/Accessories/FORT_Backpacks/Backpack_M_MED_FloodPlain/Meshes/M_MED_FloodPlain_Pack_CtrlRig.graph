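# Control Rig for the FloodPlain backpack cosmetic that adds procedural secondary motion to its pan, forceps and sugar bones.
# The forward solve runs a two-step sequence of local-space constraints on the pack skeleton.
# First, pan_base_jnt aims its negative X axis at dyn_dno_all without maintaining offset.
# Then dyn_Sugar_top1_jnt partially follows the position of dyn_forceps_jnt with its offset maintained; no other bones are driven here.
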ControlRigBlueprint /BRCosmetics/Accessories/FORT_Backpacks/Backpack_M_MED_FloodPlain/Meshes/M_MED_FloodPlain_Pack_CtrlRig
graphs [
 RigVMFunctionLibrary {links [], nodes []}
 RigVMModel {nodes [
   RigUnit_BeginExecution = &RigUnit_BeginExecution()
   SequenceAggregate = &RigVMFunction_Sequence(ExecuteContext=@RigUnit_BeginExecution.ExecutePin)
   &RigUnit_AimConstraintLocalSpaceOffset(
    AdvancedSettings=(DebugSettings=(Scale=10,WorldOffset=(Rotation=(X=0,Y=0,Z=0,W=1),Translation=(X=0,Y=0,Z=0),Scale3D=(X=1,Y=1,Z=1))),RotationOrderForFilter=XZY)
    AimAxis=(X=-1,Y=0,Z=0)
    Child=(Type=Bone,Name="pan_base_jnt")
    Filter=(bX=True,bY=True,bZ=True)
    Parents=((Item=(Type=Bone,Name="dyn_dno_all"),Weight=1))
    Weight=1
    WorldUp=(Target=(X=-1,Y=0,Z=0),Space=(Type=None,Name="None"))
    bMaintainOffset=false
    ExecutePin=@SequenceAggregate.A
   )
   /Script/ControlRig.RigUnit_PositionConstraintLocalSpaceOffset(
    Child=(Type=Bone,Name="dyn_Sugar_top1_jnt")
    Filter=(bX=True,bY=True,bZ=True)
    Parents=((Item=(Type=Bone,Name="dyn_forceps_jnt"),Weight=0.25))
    Weight=0.85
    bMaintainOffset=true
    ExecutePin=@SequenceAggregate.B
   ) {is=struct RigVMUnitNode}
  ]}
]
hierarchy [
 root {loc=<0 0 0>, rot=<0 0 0 1>}
 pack {parent=root, loc=<0 -0.7316 0>, rot=<0 0 0 1>}
 pan_base_offset_jnt1 {parent=pack, loc=<-1.7492 -9.1745 -6.8224>, rot=<-0.0941 0.5982 0.1236 -0.7862>}
 pan_base_jnt {parent=pan_base_offset_jnt1, loc=<0 0 0>, rot=<0 0 0 1>}
 dyn_pan_base_02_jnt {parent=pan_base_jnt, loc=<1.0459 0 0>, rot=<0 0 0 1>}
 dyn_forceps_offset_jnt {parent=dyn_pan_base_02_jnt, loc=<10.0196 -2.2462 6.2355>, rot=<-0.2704 -0.3549 0.7535 -0.483>}
 dyn_forceps_jnt {parent=dyn_forceps_offset_jnt, loc=<0 0 0>, rot=<0 0 0 1>}
 dyn_Sugar_top1_offset_jnt {parent=dyn_pan_base_02_jnt, loc=<7.2764 -4.2376 3.193>, rot=<0 0 0 1>}
 dyn_Sugar_top1_jnt {parent=dyn_Sugar_top1_offset_jnt, loc=<0 0 0>, rot=<0 0 0 1>}
 dyn_Sugar_top2_offset_jnt {parent=dyn_pan_base_02_jnt, loc=<7.2062 -2.081 5.8962>, rot=<0 0 0 1>}
 dyn_Sugar_top2_jnt {parent=dyn_Sugar_top2_offset_jnt, loc=<0 0 0>, rot=<0 0 0 1>}
 dyn_Sugar_top3_offset_jnt {parent=dyn_pan_base_02_jnt, loc=<6.91 -1.9162 -0.6457>, rot=<0 0 0 1>}
 dyn_Sugar_top3_jnt {parent=dyn_Sugar_top3_offset_jnt, loc=<0 0 0>, rot=<0 0 0 1>}
 dyn_Sugar_mid1_offset_jnt {parent=dyn_pan_base_02_jnt, loc=<5.5405 -0.1158 0.4389>, rot=<-0.7145 0.0295 -0.0176 -0.6988>}
 dyn_Sugar_mid1_jnt {parent=dyn_Sugar_mid1_offset_jnt, loc=<0 0 0>, rot=<0 0 0 1>}
 dyn_Sugar_mid2_offset_jnt {parent=dyn_pan_base_02_jnt, loc=<4.1908 -1.7317 0.4857>, rot=<-0.7146 0.1455 0.0771 -0.6798>}
 dyn_Sugar_mid2_jnt {parent=dyn_Sugar_mid2_offset_jnt, loc=<0 0 0>, rot=<0 0 0 1>}
 dyn_Sugar_mid3_offset_jnt {parent=dyn_pan_base_02_jnt, loc=<5.847 3.9961 -1.005>, rot=<-0.6793 -0.228 -0.1374 -0.6839>}
 dyn_Sugar_mid3_jnt {parent=dyn_Sugar_mid3_offset_jnt, loc=<0 0 0>, rot=<0 0 0 1>}
 sugra_cube_01_offset_jnt {parent=dyn_pan_base_02_jnt, loc=<9.6963 6.2837 -0.7311>, rot=<0 0 0 1>}
 sugra_cube_01_jnt {parent=sugra_cube_01_offset_jnt, loc=<0 0 0>, rot=<0 0 0 1>}
 sugra_cube_02_offset_jnt {parent=dyn_pan_base_02_jnt, loc=<8.7741 2.2848 -5.7504>, rot=<0 0 0 1>}
 sugra_cube_02_jnt {parent=sugra_cube_02_offset_jnt, loc=<0 0 0>, rot=<0 0 0 1>}
 sugra_cube_03_offset_jnt {parent=dyn_pan_base_02_jnt, loc=<8.101 -5.622 -2.3934>, rot=<0 0 0 1>}
 sugra_cube_03_jnt {parent=sugra_cube_03_offset_jnt, loc=<0 0 0>, rot=<0 0 0 1>}
 sugra_cube_04_offset_jnt {parent=dyn_pan_base_02_jnt, loc=<8.9722 5.026 3.5081>, rot=<0 0 0 1>}
 sugra_cube_04_jnt {parent=sugra_cube_04_offset_jnt, loc=<0 0 0>, rot=<0 0 0 1>}
 sugra_cube_05_offset_jnt {parent=dyn_pan_base_02_jnt, loc=<8.194 2.0141 6.347>, rot=<0 0 0 1>}
 sugra_cube_05_jnt {parent=sugra_cube_05_offset_jnt, loc=<0 0 0>, rot=<0 0 0 1>}
 sugra_cube_06_offset_jnt {parent=dyn_pan_base_02_jnt, loc=<9.3557 2.27 -1.7409>, rot=<0 0 0 1>}
 sugra_cube_06_jnt {parent=sugra_cube_06_offset_jnt, loc=<0 0 0>, rot=<0 0 0 1>}
 sugra_cube_07_offset_jnt {parent=dyn_pan_base_02_jnt, loc=<7.2794 -1.897 -5.3064>, rot=<0 0 0 1>}
 sugra_cube_07_jnt {parent=sugra_cube_07_offset_jnt, loc=<0 0 0>, rot=<0 0 0 1>}
 dyn_dno_all {parent=pack, loc=<0 0 -9.4429>, rot=<-0.0851 0.7659 0.0872 -0.6314>}
]
