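# Physics asset for the M_MED_FloodPlain backpack cosmetic, giving capsule bodies to its pan, sugar cube pieces and forceps for small secondary motion.
# The sugar and forceps bodies each constrain directly to pan_base_jnt in a star layout, with angular limits of only a few degrees.
# The pan base is itself constrained to an offset joint with swing2 and twist locked, while the forceps lock swing2 and have a wider swing1.
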PhysicsAsset /BRCosmetics/Accessories/FORT_Backpacks/Backpack_M_MED_FloodPlain/Meshes/M_MED_FloodPlain_Pack_Physics
bodies {
 dyn_Sugar_mid1_jnt [{shapeType=Capsule, length=1, radius=2, shapeName=dyn_Sugar_mid1_jnt_capsule}]
 dyn_Sugar_mid2_jnt [{shapeType=Capsule, length=1, radius=2, shapeName=dyn_Sugar_mid2_jnt_capsule}]
 dyn_Sugar_mid3_jnt [{shapeType=Capsule, length=1, radius=2, shapeName=dyn_Sugar_mid3_jnt_capsule}]
 dyn_Sugar_top1_jnt [{shapeType=Capsule, length=1, radius=2, shapeName=dyn_Sugar_top1_jnt_capsule}]
 dyn_Sugar_top2_jnt [{shapeType=Capsule, length=1, radius=2, shapeName=dyn_Sugar_top2_jnt_capsule}]
 dyn_Sugar_top3_jnt [{shapeType=Capsule, length=1, radius=2, shapeName=dyn_Sugar_top3_jnt_capsule}]
 dyn_forceps_jnt [{
   shapeType=Capsule
   length=0.4423
   radius=3.7388
   shapeName=dyn_forceps_jnt_capsule
   center=<-3.9281 0.2783 0.007>
   rotation=<rot 0.0108 1.3502 -89.9999>
  }]
 pan_base_jnt [
  {shapeType=Capsule, length=1, radius=1}
  {shapeType=Capsule, length=7.3098, radius=3.0661, shapeName=pan_base_jnt_capsule, center=<4.4767 -6.0043 -0.4519>, rotation=<rot 5 -0.0002 0>}
 ]
 pan_base_offset_jnt1 [{shapeType=Capsule, length=0.505, radius=0.505, shapeName=pan_base_offset_jnt1_capsule, rotation=<rot 0 -90 0>}]
 sugra_cube_01_jnt [{shapeType=Capsule, length=1, radius=2, shapeName=sugra_cube_01_jnt_capsule}]
 sugra_cube_02_jnt [{shapeType=Capsule, length=1, radius=2, shapeName=sugra_cube_02_jnt_capsule}]
 sugra_cube_03_jnt [{shapeType=Capsule, length=1, radius=2, shapeName=sugra_cube_03_jnt_capsule}]
 sugra_cube_04_jnt [{shapeType=Capsule, length=1, radius=2, shapeName=sugra_cube_04_jnt_capsule}]
 sugra_cube_05_jnt [{shapeType=Capsule, length=1, radius=2, shapeName=sugra_cube_05_jnt_capsule}]
 sugra_cube_06_jnt [{shapeType=Capsule, length=1, radius=2, shapeName=sugra_cube_06_jnt_capsule, center=<0 0 0.3164>}]
 sugra_cube_07_jnt [{shapeType=Capsule, length=1, radius=2, shapeName=sugra_cube_07_jnt_capsule}]
}
constraints [
 {
  bone1Name=pan_base_jnt
  bone2Name=pan_base_offset_jnt1
  swing1LimitDegrees=0.3
  swing2LimitDegrees=2
  swing2Motion=Locked
  twistLimitDegrees=2
  twistMotion=Locked
 }
 {bone1Name=dyn_Sugar_mid2_jnt, bone2Name=pan_base_jnt, swing1LimitDegrees=3, swing2LimitDegrees=2, twistLimitDegrees=2}
 {bone1Name=dyn_Sugar_mid1_jnt, bone2Name=pan_base_jnt, swing1LimitDegrees=3, swing2LimitDegrees=2, twistLimitDegrees=2}
 {bone1Name=dyn_Sugar_mid3_jnt, bone2Name=pan_base_jnt, swing1LimitDegrees=3, swing2LimitDegrees=2, twistLimitDegrees=2}
 {bone1Name=dyn_Sugar_top1_jnt, bone2Name=pan_base_jnt, swing1LimitDegrees=3, swing2LimitDegrees=2, twistLimitDegrees=2}
 {bone1Name=dyn_Sugar_top2_jnt, bone2Name=pan_base_jnt, swing1LimitDegrees=3, swing2LimitDegrees=2, twistLimitDegrees=2}
 {bone1Name=dyn_Sugar_top3_jnt, bone2Name=pan_base_jnt, swing1LimitDegrees=3, swing2LimitDegrees=2, twistLimitDegrees=2}
 {bone1Name=sugra_cube_01_jnt, bone2Name=pan_base_jnt, swing1LimitDegrees=3, swing2LimitDegrees=2, twistLimitDegrees=2}
 {bone1Name=sugra_cube_02_jnt, bone2Name=pan_base_jnt, swing1LimitDegrees=3, swing2LimitDegrees=2, twistLimitDegrees=2}
 {bone1Name=sugra_cube_03_jnt, bone2Name=pan_base_jnt, swing1LimitDegrees=3, swing2LimitDegrees=2, twistLimitDegrees=2}
 {bone1Name=sugra_cube_04_jnt, bone2Name=pan_base_jnt, swing1LimitDegrees=3, swing2LimitDegrees=2, twistLimitDegrees=2}
 {bone1Name=sugra_cube_05_jnt, bone2Name=pan_base_jnt, swing1LimitDegrees=3, swing2LimitDegrees=2, twistLimitDegrees=2}
 {bone1Name=sugra_cube_06_jnt, bone2Name=pan_base_jnt, swing1LimitDegrees=4, swing2LimitDegrees=2, twistLimitDegrees=2}
 {bone1Name=sugra_cube_07_jnt, bone2Name=pan_base_jnt, swing1LimitDegrees=3, swing2LimitDegrees=2, twistLimitDegrees=2}
 {bone1Name=dyn_forceps_jnt, bone2Name=pan_base_jnt, swing1LimitDegrees=6.9, swing2LimitDegrees=1, swing2Motion=Locked, twistLimitDegrees=1}
]
