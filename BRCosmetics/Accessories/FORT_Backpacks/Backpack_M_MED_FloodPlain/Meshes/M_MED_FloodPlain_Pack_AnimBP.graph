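# Animation Blueprint for the FloodPlain backpack cosmetic, adding physics-driven secondary motion that reacts to the wearer's movement.
# It pulls dyn_dno_all toward the owner mesh's world location, caches that pose, then blends two RigidBody setups for front end and gameplay.
# A Control Rig follows, with additive pan_base_jnt offsets for DBNO, floating skydive and grounded states.
# The event graph sets gravity and inertia scales for ground, falling, skydive and parachute states.
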AnimBlueprint /BRCosmetics/Accessories/FORT_Backpacks/Backpack_M_MED_FloodPlain/Meshes/M_MED_FloodPlain_Pack_AnimBP
target_skeleton=&/M_MED_FloodPlain_Pack_Skeleton
graphs [
 AnimGraph {nodes [
   n1 = Get Gravity Override()
   n2 = RigidBody!(ComponentPose=Local To Component(LocalPose=Use cached pose 'PreSim'()), Alpha, OverrideWorldGravity:Vector=@n1, ExternalForce) -> (Pose) {
    bClampLinearTranslationLimitToRefPose=true
    bOverrideWorldGravity=true
    overridePhysicsAsset=&/M_MED_FloodPlain_Pack_Physics
    simulationSpace=BaseBoneSpace
    alphaBoolBlend {blendOption=Linear}
    alphaScaleBiasClamp {interpSpeedIncreasing=2}
    baseBoneRef {boneName=root}
    componentAppliedLinearAccClamp=<3000 3000 3000>
    componentLinearAccScale=<0.4 0.3 0.3>
    componentLinearVelScale=<0.4 0.3 0.3>
    simSpaceSettings {
     maxAngularAcceleration=1000
     maxAngularVelocity=3
     maxLinearAcceleration=700
     maxLinearVelocity=500
     velocityScaleZ=0
     worldAlpha=0.25
     externalAngularVelocity {z=0.1}
     externalLinearDragV=<0.2 0.2>
    }
   }
   n3 = Get Component Linear Vel Scale() -> (Component Linear Vel Scale:Vector)
   n4 = Get Component Linear Acc Scale() -> (Component Linear Acc Scale:Vector)
   n5 = RigidBody!(
    ComponentPose=Local To Component(LocalPose=Use cached pose 'PreSim'())
    Alpha
    OverrideWorldGravity=@n1
    ExternalForce
    ComponentLinearAccScale:Vector=@n4
    ComponentLinearVelScale:Vector=@n3
   ) -> (Pose) {
    bOverrideWorldGravity=true
    cachedBoundsScale=1
    overridePhysicsAsset=&M_MED_FloodPlain_Pack_Physics
    alphaBoolBlend {blendOption=Linear}
    componentAppliedLinearAccClamp=<85 30 230>
    componentLinearAccScale=<1 0.5 1>
    componentLinearVelScale=<1 0.5 1>
    overrideWorldGravity {z=-100}
    simSpaceSettings {maxAngularAcceleration=100, maxAngularVelocity=5, maxLinearAcceleration=1340, maxLinearVelocity=15, velocityScaleZ=10}
   }
   n6 = Control Rig(Source=Blend Poses by bool(BlendPose_0=Component To Local(ComponentPose=@n2), BlendPose_1=Component To Local(ComponentPose=@n5), bActiveValue=Get bIsInFrontEnd()) {blendType=Linear}) {
    alphaBoolBlend {blendOption=Linear}
    controlRigAssetReference {blueprintRigClass=&/M_MED_FloodPlain_Pack_CtrlRig.M_MED_FloodPlain_Pack_CtrlRig_C}
   }
   n7 = "Transform (Modify) Bone - Bone: pan_base_jnt"!(
    ComponentPose=Local To Component(LocalPose=@n6)
    bAlphaBoolEnabled:bool=Get DBNO_and_Skydive() -> (DBNO_and_Skydive:bool)
    Translation="-4,0,-4"
    Rotation="0,0,0"
    Scale="1,1,1"
   ) -> (Pose) {
    alphaInputType=Bool
    translationMode=BMM_Additive
    translationSpace=BCS_BoneSpace
    alphaBoolBlend {blendOption=Linear, blendOutTime=1}
    boneToModify {boneName=pan_base_jnt}
    translation {x=-4, z=-4}
   }
   n8 = "Transform (Modify) Bone - Bone: pan_base_jnt"(
    ComponentPose=@n7
    bAlphaBoolEnabled=Get bIsOnGround() -> (bIsOnGround:bool)
    Translation="0,0,0"
    Rotation="-2.628,28.239,0.000"
    Scale="1,1,1"
   ) {
    alphaInputType=Bool
    rotationMode=BMM_Additive
    rotationSpace=BCS_BoneSpace
    alphaBoolBlend {blendInTime=0.5, blendOption=Linear, blendOutTime=1}
    boneToModify {boneName=pan_base_jnt}
    rotation=<rot -2.628 28.239 0>
   }
   Output Pose(Result=Component To Local(ComponentPose=@n8)) {layerGroup=None}
   n9 = "Transform (Modify) Bone - Bone: dyn_dno_all"!(
    ComponentPose=Local To Component(LocalPose=Input Pose())
    Alpha=1
    Translation_X:double=Get Return Value X() -> (Return Value X:double)
    Translation_Y:double=Get Return Value Y() -> (Return Value Y:double)
    Translation_Z:double=Get Return Value Z() -> (Return Value Z:double)
    Rotation="0,0,0"
    Scale="1,1,1"
   ) -> (Pose) {
    rotationSpace=BCS_WorldSpace
    translationMode=BMM_Replace
    translationSpace=BCS_WorldSpace
    alphaBoolBlend {blendOption=Linear}
    alphaScaleBiasClamp {scale=0.1}
    boneToModify {boneName=dyn_dno_all}
    translation=<2180.022 0.992>
   }
   Save cached pose 'PreSim'(Pose=Component To Local(ComponentPose=@n9)) {cacheName=PreSim}
   Reroute Node(InputPin:Vector=@n4) -> (OutputPin:Vector)
   Reroute Node(InputPin=@n3)
  ]}
 EventGraph {nodes [
   Event BlueprintUpdateAnimation()
   > n1 = Sequence!(execute) -> (then_0, then_1, then_2:Exec, then_3:Exec)
   n2 = Gravity Override State Function DEPRECATED(
    execute=@n1.then_2
    CharacterPartAnimInstance=Self-Reference()
    GameTimeInSeconds=GetGameTimeInSeconds()
    GravityOverrideParams=Get Gravity Override Params()
   )
   > Set Gravity Override(Gravity Override=@n2)
   > Set DBNO_and_Skydive(DBNO_and_Skydive:bool=OR Boolean(A=Get bIsDBNO(), B=Get bIsSkydiveFloating())) -> (Output_Get:bool)
   n3 = Get bIsOnGround()
   n4 = Branch(execute=@n1.then_3, Condition=@n3)
   > Set Gravity Override(Gravity Override="0,0,-180")
   > Set Component Linear Acc Scale(Component Linear Acc Scale:Vector="0.12,0.1,0.12") -> (Output_Get:Vector)
   > Set Component Linear Vel Scale(Component Linear Vel Scale:Vector="0.1,0.1,0.1") -> (Output_Get:Vector)
   n5 = Branch(execute=@n4.else, Condition=OR Boolean(A=Get bIsFalling(), B=Get bIsSkyDiving()))
   > n6 = Branch(Condition=OR Boolean(A=Get bIsSkydiveDiveMode(), B=Get bIsParachuteOpened()))
   > n7 = Branch(Condition=Get bIsParachuteOpened())
   > n8 = Branch(Condition="float > float"(A=Get PawnVelocityZ(), B=45))
   > Set Gravity Override(Gravity Override="0,0,-170")
   > Set Component Linear Acc Scale(Component Linear Acc Scale="0.2,0.4,0.35")
   > Set Component Linear Vel Scale(Component Linear Vel Scale="0.1,0.1,0.1")
   Set Gravity Override(execute=[@n8.else, @n7.else], Gravity Override="0,0,-150")
   > Set Component Linear Acc Scale(Component Linear Acc Scale="0.12,0.1,0.12")
   > Set Component Linear Vel Scale(Component Linear Vel Scale="0.1,0.1,0.1")
   Set Gravity Override(execute=@n6.else, Gravity Override="0,0,-210")
   > Set Component Linear Acc Scale(Component Linear Acc Scale="0.12,0.1,0.12")
   > Set Component Linear Vel Scale(Component Linear Vel Scale="0.1,0.1,0.1")
   Set Gravity Override(execute=@n5.else, Gravity Override="0,0,-145")
   > Set Component Linear Acc Scale(Component Linear Acc Scale="0.12,0.25,0.1")
   > Set Component Linear Vel Scale(Component Linear Vel Scale="0.1,0.1,0.1")
   n9 = Get World Location!(self:Scene Component Object Reference=GetOwnerSkeletalMeshComponent(self=Self-Reference())) -> (ReturnValue_X:double, ReturnValue_Y:double, ReturnValue_Z:double) {is=title K2Node_CallFunction}
   Set Return Value X(execute=@n1.then_0, Return Value X:double=@n9) -> (Output_Get:double)
   > Set Return Value Y(Return Value Y:double=@n9.ReturnValue_Y) -> (Output_Get:double)
   > Set Return Value Z(Return Value Z:double=@n9.ReturnValue_Z) -> (Output_Get:double)
   OR Boolean(A=@n3, B=Get bIsFalling())
  ]}
]
variables [
 Activate Ctrl Rig {type=boolean}
 activate_ctrl_rig {type=boolean, default=true}
 Component Linear Acc Scale {type=Vector, default=<1 0.5 1>}
 Component Linear Vel Scale {type=Vector, default=<1 0.5 1>}
 DBNO_and_Skydive {type=boolean}
 Gravity Override {type=Vector}
 Gravity Override Params {type=GravityOverrideParamsStruct}
 is flight {type=boolean}
 Is Flight_0 {type=boolean}
 Return Value X {type=number}
 Return Value Y {type=number}
 Return Value Z {type=number}
 Target {type=CustomCharacterPartAnimInstance, default=null}
]
